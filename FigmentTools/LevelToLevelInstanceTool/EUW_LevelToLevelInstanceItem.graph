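# Editor utility list-row widget for the Level To Level Instance tool, showing one level asset and its conversion settings.
# When a list item object is set, it stores the LevelToLevelInstanceToolItem, binds its change event and refreshes.
# Checkboxes and the text box write inclusion, custom-name toggle and target package name into the item.
# UpdateView mirrors the item back, making the name editable only when custom naming is enabled.
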EditorUtilityWidgetBlueprint /FigmentTools/LevelToLevelInstanceTool/EUW_LevelToLevelInstanceItem
graphs [
 EventGraph {nodes [
   n1 = "On Check State Changed (checkbox_useCustom)"!() -> (OutputDelegate, then, bIsChecked:bool) {is=title K2Node_ComponentBoundEvent}
   > SetUseCustomTargetPackageName!(
    execute
    self:Level To Level Instance Tool Item Object Reference=Get ItemData() -> (ItemData:Level To Level Instance Tool Item Object Reference)
    bUse:bool=@n1.bIsChecked
   ) -> (then)
   > UpdateView()
   n2 = "On Check State Changed (checkbox_useChecked)"!() -> (OutputDelegate, then, bIsChecked:bool) {is=title K2Node_ComponentBoundEvent}
   > Set bIsChecked!(execute, bIsChecked:bool=@n2.bIsChecked, self:Level To Level Instance Tool Item Object Reference=Get ItemData()) -> (then, Output_Get:bool)
   > UpdateView()
   n3 = "On Text Committed (EditorUtilityEditableTextBox_112)"!() -> (OutputDelegate, then, Text:"Text (by ref)", CommitMethod:ETextCommit Enum=Default) {is=title K2Node_ComponentBoundEvent}
   > Branch(Condition=Get bUseCustomTargetPackageName!(self:Level To Level Instance Tool Item Object Reference=Get ItemData()) -> (bUseCustomTargetPackageName:bool))
   > Set TargetPackageName!(
    execute
    TargetPackageName:String="To String (Text)"(InText=@n3.Text)
    self:Level To Level Instance Tool Item Object Reference=Get ItemData()
   ) -> (then, Output_Get:String)
   > UpdateView()
   Event PreConstruct()
   > n4 = Construct Object from Class(Class=/Script/FigmentTools.LevelToLevelInstanceToolItem) -> (ReturnValue:Level To Level Instance Tool Item Object Reference)
   > n5 = Set ItemData(ItemData:Level To Level Instance Tool Item Object Reference=@n4) -> (Output_Get:Level To Level Instance Tool Item Object Reference)
   > Set TargetPackageName(TargetPackageName=Target Package Name, self=@n5)
   n6 = Get ItemData()
   > Set bUseCustomTargetPackageName!(execute, bUseCustomTargetPackageName:bool=false, self:Level To Level Instance Tool Item Object Reference=@n6) -> (then, Output_Get:bool)
   > Set bIsChecked(bIsChecked=true, self=@n6)
   > UpdateView()
   n7 = Event OnListItemObjectSet()
   > n8 = Cast To LevelToLevelInstanceToolItem!(execute, Object=@n7.ListItemObject) -> (then, CastFailed, AsLevel to Level Instance Tool Item:Level To Level Instance Tool Item Object Reference) {is=title K2Node_DynamicCast}
   > n9 = Set ItemData(ItemData=@n8)
   > Bind Event to On Item Changed(self:Level To Level Instance Tool Item Object Reference=@n9, Delegate=Create Event()) {is=title K2Node_AddDelegate}
   > UpdateView()
   Event Construct()
   Event Tick()
  ]}
 UpdateView {nodes [
   UpdateView() {is=title K2Node_FunctionEntry}
   n1 = Get LevelAsset!(self:Level To Level Instance Tool Item Object Reference=Get ItemData()) -> (LevelAsset:Asset Data Structure)
   > "SetText (Text)"(
    self=Get text_packageName() -> (text_packageName:Text Object Reference)
    InText="To Text (Name)"!(InName:Name=Break Asset Data(AssetData=@n1)) -> (ReturnValue:Text)
   )
   > SetIsChecked!(
    execute
    self:Check Box Object Reference=Get checkbox_useChecked() -> (checkbox_useChecked:Editor Utility Check Box Object Reference)
    InIsChecked:bool=Get bIsChecked!(self:Level To Level Instance Tool Item Object Reference=Get ItemData()) -> (bIsChecked:bool)
   ) -> (then)
   > SetIsChecked(
    self=Get checkbox_useCustom() -> (checkbox_useCustom:Editor Utility Check Box Object Reference)
    InIsChecked=Get bUseCustomTargetPackageName(self=Get ItemData())
   )
   n2 = Get ItemData()
   > "SetText (Text Box)"!(
    execute
    self:Text Box Object Reference=Get EditorUtilityEditableTextBox_112() -> (EditorUtilityEditableTextBox_112:Editor Utility Editable Text Box Object Reference)
    InText:Text="To Text (String)"(InString=Get TargetPackageName!(self:Level To Level Instance Tool Item Object Reference=@n2) -> (TargetPackageName:String))
   ) -> (then)
   n3 = Get ItemData()
   > "SetIsReadOnly (Editable Text)"!(
    execute
    self:Text Box Object Reference=Get EditorUtilityEditableTextBox_112()
    bReadOnly:bool=NOT Boolean(A=Get bUseCustomTargetPackageName(self=@n3))
   ) -> (then)
  ]}
]
variables [ItemData {type=LevelToLevelInstanceToolItem, default=null}]
widgets {
 info {rootWidgetClass=&HorizontalBox, widgetCount=8}
 widgets [
  HorizontalBox_61 {widgetClassPath=&HorizontalBox, widget=HorizontalBox_61}
  SizeBox {
   widgetClassPath=&SizeBox
   parent=HorizontalBox_61
   widget=SizeBox
   properties {bOverride_WidthOverride=true, maxDesiredWidth=25, widthOverride=20}
   slot {class=HorizontalBoxSlot}
  }
  checkbox_useChecked {
   widgetClassPath=&EditorUtilityCheckBox
   bIsVariable=true
   parent=SizeBox
   widget=checkbox_useChecked
   slot {class=SizeBoxSlot}
  }
  text_packageName {
   widgetClassPath=&TextBlock
   bIsVariable=true
   parent=HorizontalBox_61
   widget=text_packageName
   properties {
    text=Text Block
    font {
     bForceMonospaced=false
     bMaterialIsStencil=false
     fontMaterial=null
     fontObject=&NotoSans
     letterSpacing=0
     monospacedWidth=1
     size=8.25
     skewAmount=0
     typefaceFontName=NotoSans-Regular
     outlineSettings {
      bApplyOutlineToDropShadows=false
      bMiteredCorners=false
      bSeparateFillAlpha=false
      outlineMaterial=null
      outlineSize=0
      outlineColor=<rgba 0 0 0 1>
     }
    }
   }
   slot {class=HorizontalBoxSlot, properties {verticalAlignment=VAlign_Center, size {sizeRule=Fill, value=1}}}
  }
  Spacer_138 {widgetClassPath=&Spacer, parent=HorizontalBox_61, widget=Spacer_138, properties {size=<10 1>}, slot {class=HorizontalBoxSlot}}
  EditorUtilityEditableTextBox_112 {
   widgetClassPath=&EditorUtilityEditableTextBox
   bIsVariable=true
   parent=HorizontalBox_61
   widget=EditorUtilityEditableTextBox_112
   slot {class=HorizontalBoxSlot, properties {size {sizeRule=Fill, value=1}}}
  }
  SizeBox_148 {
   widgetClassPath=&SizeBox
   parent=HorizontalBox_61
   widget=SizeBox_148
   properties {bOverride_WidthOverride=true, maxDesiredWidth=25, widthOverride=80}
   slot {class=HorizontalBoxSlot}
  }
  checkbox_useCustom {
   widgetClassPath=&EditorUtilityCheckBox
   bIsVariable=true
   parent=SizeBox_148
   widget=checkbox_useCustom
   slot {class=SizeBoxSlot}
  }
 ]
}
